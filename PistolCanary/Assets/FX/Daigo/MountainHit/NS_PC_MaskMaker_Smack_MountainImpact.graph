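# Niagara impact effect, named for a MaskMaker smack into a mountain, combining an energy flash with snow and debris.
# Burst (light, sprite, mesh; skipped in water), radial plasma and streaks form the brief flash.
# GPU mesh fragments spawn from skeletal-mesh sampling, fly out under drag and gravity, and scale with camera distance.
# Lingering and rising SubUV snow sprites follow, with time-of-day functions adjusting snow tint and flash brightness.
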NiagaraSystem /PistolCanary/Assets/FX/Daigo/MountainHit/NS_PC_MaskMaker_Smack_MountainImpact
dependencies {
 usedDataInterfaces [
  &NiagaraDataInterfaceSpriteRendererInfo
  &NiagaraDataInterfaceCurve
  &NiagaraDataInterfaceColorCurve
  &NiagaraDataInterfaceCamera
  &NiagaraDataInterfaceSkeletalMesh
  &NiagaraDataInterfaceVector2DCurve
 ]
 usedDynamicInputs [
  &MergeHelper_VectorSwitch
  &SimulationPosition
  &RandomRangeVector2D
 ]
 usedModules [&SystemState]
}
emitters {
 Burst {
  emitterUpdateScript [
   &EmitterState()
   &SpawnBurst_Instantaneous(
    Spawn Count=1
    Spawn Probability=&MakeCustomFloatFromBool(Bool=$User.bIsInWater, True Float=0, False Float=1)
    Use Spawn Probability=true
   )
  ]
  particleSpawnScript [
   &InitializeParticle(
    Color=<rgba 5 2.8009 0.3646 1>
    Lifetime=0.3
    Mesh Scale Mode=::NewEnumerator3
    Mesh Uniform Scale=1.2
    Sprite Size Mode=::NewEnumerator3
    Uniform Sprite Size=10000
    UsePositionOffset=false
    Write Material Random=true
   )
   /SetVariables(Particles.LightPos:NiagaraPosition=&Add_Vector(A=Linked() {linkedVariable {name=Particles.Position, type=NiagaraPosition}}, B=<100 0 0>)) {isSetParameters=true}
   /SetVariables(Particles.MeshColor:LinearColor=<rgba 12 9.4975 4.5867 1>, Particles.MeshLifetime=0.15) {isSetParameters=true}
  ]
  particleUpdateScript [
   &ParticleState()
   &ScaleSpriteSize(Scale Factor=&Vector2DFromFloat(Value=&FloatFromCurve(
      FloatCurve=DataInterface() {propertyValues {Curve {
         defaultValue=0.5
         keys [
          {arriveTangent=4.1785, interpMode=RCIM_Cubic, leaveTangent=4.1785, tangentMode=RCTM_User, time=0, value=0}
          {arriveTangent=1.1179, interpMode=RCIM_Cubic, leaveTangent=1.1179, tangentMode=RCTM_User, time=0.2042, value=0.6228}
          {arriveTangent=0.0224, interpMode=RCIM_Cubic, leaveTangent=0.0224, tangentMode=RCTM_User, time=1, value=1}
         ]
        }}}
      Scale Curve=1.25
     )))
   &ScaleMeshSize(Scale Factor=&VectorFromFloat(Value=&FloatFromCurve(
      FloatCurve=DataInterface() {propertyValues {Curve {keys [{arriveTangent=9.901, interpMode=RCIM_Cubic, leaveTangent=9.901, tangentMode=RCTM_User, time=0, value=0.4}, <at 1 6 RCIM_Cubic>]}}}
      Scale Curve=&ClampFloat(Float=&Ceil_Float(Float=&Subtract_Float(A=$Particles.MeshLifetime, B=$Particles.Age)))
     )))
   &ScaleColor(
    Scale Alpha=&Multiply_Float(A=0.2, B=&FNTimeOfDayScalar(Night Scale=0.3, Morning Scale=0.6, Evening Scale=0.6))
    Scale RGB=&VectorFromFloat(Value=&FloatFromCurve(FloatCurve=DataInterface() {propertyValues {Curve {keys [{interpMode=RCIM_Cubic, tangentMode=RCTM_User, time=0, value=1}, {interpMode=RCIM_Cubic, tangentMode=RCTM_User, time=1, value=0}]}}}))
    ScaleA=true
    ScaleRGB=true
   )
   &DynamicMaterialParameters(
    Index 0 Param 1=1
    Index 0 Param 2=1
    Index 0 Param 3=1
    Index 0 Param 4=1
    Module.Param0WriteEnabled=true
    Module.Param1WriteEnabled=true
    Module.Param2WriteEnabled=true
    Module.Param3WriteEnabled=true
   )
   &FN_Light_Attributes(
    Color=<rgba 0.5 0.4031 0.205 1>
    Create Color Attribute=true
    Create Exponent Attribute=true
    Create Position Attribute=true
    Create Radius Attribute=true
    Exponent=2
    Position=&Add_Vector(A=Linked() {linkedVariable {name=Particles.Initial.Position, type=NiagaraPosition}}, B=<0 0 25>)
    Radius=4000
    Scale Color=&FloatFromCurve(
     FloatCurve=DataInterface() {propertyValues {Curve {keys [
         {interpMode=RCIM_Cubic, leaveTangent=-4.7325, leaveTangentWeight=0.9776, tangentMode=RCTM_Break, time=0.1, value=2}
         {arriveTangent=-0.476, interpMode=RCIM_Cubic, tangentMode=RCTM_Break, time=1, value=0}
        ]}}}
     Scale Curve=&FNTimeOfDayScalar(Night Scale=3, Morning Scale=4, Day Scale=5, Evening Scale=4)
    )
   )
   /SetVariables(Particles.MeshColor:LinearColor=&MultiplyLinearColor(
     Linear Color=$Particles.Initial.MeshColor
     Scale Factor=&MakeVector4FromVectorAndFloat(
      "Vector (RGB)"=<1 1 1>
      "Float (Alpha)"=&FloatFromCurve(
       FloatCurve=DataInterface() {propertyValues {Curve {keys [<at 0.4 1 RCIM_Cubic>,<at 1 0>]}}}
       Scale Curve=&Multiply_Float(A=1, B=&FNTimeOfDayScalar(Night Scale=0.1, Morning Scale=0.5, Evening Scale=0.5))
      )
     )
    )) {isSetParameters=true}
  ]
  rendererClasses [
   &NiagaraLightRendererProperties
   &NiagaraSpriteRendererProperties
   &NiagaraMeshRendererProperties
  ]
  renderers [
   {rendererClass=&NiagaraLightRendererProperties, rendererIndex=0}
   {rendererClass=&NiagaraSpriteRendererProperties, rendererIndex=1}
   {rendererClass=&NiagaraMeshRendererProperties, rendererIndex=2}
  ]
 }
 Fragments {
  simTarget=GPUComputeSim
  emitterUpdateScript [
   &EmitterState()
   &SpawnBurst_Instantaneous(Spawn Count=65, Use Spawn Probability=false, UseLoopCountLimit=false)
  ]
  particleSpawnScript [
   &InitializeParticle(
    Color=&MultiplyLinearColor(
     Linear Color=<rgba 0.7837 0.8367 0.8698 1>
     Scale Factor=&MakeVector4_FromFloat(Float=&FN_FrontEndFloatModifier(Gameplay=1, Front End=0.8))
    )
    Lifetime Max=3
    Lifetime Mode=::NewEnumerator1
    Mesh Scale=&Multiply_VectorByFloat(
     Vector=&RandomRangeVector(Minimum=<1 0.5 0.5>, Maximum=<1.5 1 1>)
     Float=&Multiply_Float(
      A=&RandomRangeFloat(Minimum=0.35, Maximum=0.65)
      B=&NormalizeDistanceRange(
       Start Position:NiagaraPosition=$Particles.Position
       End Position:NiagaraPosition=&ReturnCameraPosition(Coordinate Space:ENiagaraCoordinateSpace=::Local, Camera Query:NiagaraDataInterfaceCamera=DataInterface())
       Distance:NiagaraFloat=50000
       Invert Normalized Range:NiagaraBool=false
      )
     )
    )
    Mesh Scale Mode=::NewEnumerator4
    UsePositionOffset=false
    Write Material Random=true
   )
   &SkeletalMeshLocation(
    Bone Sampling Mode=::NewEnumerator4
    Interpolate Mesh Sampling=false
    MaskSpawnGroup=false
    Orientation Sampling=::NewEnumerator1
    Sampled Position Offset=&Add_Vector(A=<0 -120 -15>, B=<0 50 0>)
    Skeletal Mesh=DataInterface() {propertyValues {
      DefaultMesh=null
      ExcludeBoneName=root
      OptionalComponentClass=null
      PreviewMesh=/BRCosmetics/Gadgets/Assets/VinderTech_GliderChute/Glider_Snowboard_Male/Meshes/Snowboard_Male_Glider
      SkinningMode=SkinOnTheFly
      SoftSourceActor=null
      SourceMode=Default
      UvSetIndex=0
      WholeMeshLOD=-1
      bExcludeBone=false
      bReadDeformedGeometry=true
      bRequireCurrentFrameData=true
      ComponentTags []
      FilteredBones [Glider_Base]
      FilteredSockets []
      MeshUserParameter {parameter {name=None, typeDefHandle {registeredTypeIndex=1}, varData []}}
      SamplingRegions []
     }}
   )
   &SphereLocation(
    MaskSpawnGroup=false
    Non Uniform Scale=<1 1 0.1>
    Override Seed=false
    Sphere Radius=40
    Surface Only=false
    UseNonUniformScale=true
    UseOffset=false
    UseSphereOrientation=false
   )
   &AddVelocityFromPoint(Module.Use Falloff Distance=false, Origin Offset=<0 -122 0>, Velocity Strength=500)
   &AddVelocityInCone(
    Cone Angle=121.511
    Cone Axis=<0 0 1>
    Module.Use Velocity Falloff On Cone Axis=false
    Velocity Strength=&RandomRangeFloat(Minimum=1200, Maximum=5000)
   ) {module=AddVelocityInCone001}
   &CameraOffset()
   &InitialMeshOrientation(
    Mesh Orientation Mode=::NewEnumerator2
    Rotation=&RandomRangeVector(Minimum=<0 0 0>, Maximum=<1 0 0>)
    Use Rotation Vector=true
   ) {module=InitialMeshOrientation001}
  ]
  particleUpdateScript [
   &Drag(Use Linear Drag=true, Use Rotational Drag=false)
   &GravityForce()
   &ParticleState()
   &ScaleMeshSize(Scale Factor=&VectorFromFloat(Value=&FloatFromCurve(FloatCurve=DataInterface() {propertyValues {Curve {
         defaultValue=1
         keys [
          {tangentMode=RCTM_User, time=0, value=1}
          {arriveTangent=-0.0405, interpMode=RCIM_Cubic, leaveTangent=-0.0405, tangentMode=RCTM_User, time=0.4, value=1}
          {arriveTangent=-3.8407, interpMode=RCIM_Cubic, leaveTangent=-8.0089, tangentMode=RCTM_Break, time=1, value=0}
         ]
        }}})))
   &ScaleSpriteSize(Scale Factor=&Vector2DFromFloat(Value=&FloatFromCurve(FloatCurve=DataInterface() {propertyValues {Curve {keys [
          {arriveTangent=6.4968, interpMode=RCIM_Cubic, leaveTangent=6.4968, time=0, value=0}
          {interpMode=RCIM_Cubic, leaveTangent=-2.1899, tangentMode=RCTM_Break, time=0.1351, value=0.878}
          {arriveTangent=-1.0152, interpMode=RCIM_Cubic, leaveTangent=-1.0152, time=1, value=0}
         ]}}}))) {enabled=false, module=ScaleSpriteSize001}
   &SpriteRotationRate(Rotation Rate=&RandomRangeFloat(Minimum=-1, Maximum=1))
   &ScaleColor(
    Scale Alpha=&Multiply_Float(
     A=&RandomRangeFloat(Minimum=0.2, Maximum=1)
     B=&FloatFromCurve(FloatCurve=DataInterface() {propertyValues {Curve {keys [
          {interpMode=RCIM_Cubic, tangentMode=RCTM_User, time=0, value=0}
          <at 0.3 1>
          {interpMode=RCIM_Cubic, tangentMode=RCTM_User, time=0.5, value=1}
          <at 0.7 1>
          {tangentMode=RCTM_User, time=1, value=0}
         ]}}})
    )
    Scale RGB=&Lerp_Vector(
     A=&MakeVectorFromLinearColorRGB(Color=<rgba 0.4165 0.6176 0.714 1>)
     B=<1 1 1>
     Alpha=&FN_FrontEndFloatModifier(Gameplay=1, Front End=0)
    )
    ScaleA=true
    ScaleRGB=true
   )
   &SolveForcesAndVelocity(Clamp Velocity=false, Limit Acceleration=false)
  ]
  rendererClasses [&NiagaraMeshRendererProperties]
  renderers [{rendererClass=&NiagaraMeshRendererProperties, rendererIndex=0}]
 }
 Radial_Plasma_Energy {
  emitterUpdateScript [
   &EmitterState()
   &SpawnBurst_Instantaneous(Spawn Count=1, Use Spawn Probability=false)
  ]
  particleSpawnScript [&InitializeParticle(
    Color Mode=::NewEnumerator0
    Lifetime=0.25
    Sprite Size Mode=::NewEnumerator3
    Uniform Sprite Size=5000
    UsePositionOffset=false
    Write Material Random=true
   )]
  particleUpdateScript [
   &ParticleState()
   &FN_Radial_Material(Base Sprite Tile U=0.5, Base Sprite Tile V=3, Depth Fade=2500, Radial Mask Offset X=0.1, Sprite Extrusion Speed=1.25)
   &Color(
    Color=&ColorFromCurve(ColorCurve=DataInterface() {propertyValues {
       AlphaCurve {keys [<at 0 1 RCIM_Cubic>,<at 1 1 RCIM_Cubic>]}
       BlueCurve {keys [<at 0.4032 0 RCIM_Cubic>,<at 1 0.561 RCIM_Cubic>]}
       GreenCurve {keys [{arriveTangent=1.1139, interpMode=RCIM_Cubic, time=0.4032, value=0.5102}, {interpMode=RCIM_Cubic, leaveTangent=1.1139, time=1, value=0.8381}]}
       RedCurve {keys [<at 0.4032 1 RCIM_Cubic>,<at 1 1 RCIM_Cubic>]}
      }})
    Scale Alpha=0.7
    Scale Color=&VectorFromFloat(Value=&FloatFromCurve(
      FloatCurve=DataInterface() {propertyValues {Curve {keys [
          {arriveTangent=-2.6449, interpMode=RCIM_Cubic, leaveTangent=-2.6449, tangentMode=RCTM_User, time=0, value=1}
          {interpMode=RCIM_Cubic, tangentMode=RCTM_User, time=1, value=0.01}
         ]}}}
      Scale Curve=20
     ))
   )
   &ScaleSpriteSize(Scale Factor=&Vector2DFromFloat(Value=&FloatFromCurve(FloatCurve=DataInterface() {propertyValues {Curve {keys [
          {arriveTangent=2.2779, interpMode=RCIM_Cubic, leaveTangent=2.2779, tangentMode=RCTM_User, time=0, value=0}
          {interpMode=RCIM_Cubic, tangentMode=RCTM_User, time=1, value=1}
         ]}}})))
  ]
  rendererClasses [&NiagaraSpriteRendererProperties]
  renderers [{rendererClass=&NiagaraSpriteRendererProperties}]
 }
 Snow_Linger {
  emitterUpdateScript [
   &EmitterState()
   &SpawnBurst_Instantaneous(Use Spawn Probability=false, UseLoopCountLimit=false)
  ]
  particleSpawnScript [
   &InitializeParticle(
    Color=&FNTimeOfDayColor(
     Night Color=<rgba 0.31 0.3011 0.4818 1>
     Morning Color=<rgba 1 1 1 1>
     Day Color=<rgba 0.8892 0.875 1 1>
     Evening Color=<rgba 0.4818 0.3425 0.4649 1>
    )
    Lifetime=2.5
    Mesh Scale=&MaskVectorBySpawnGroup(
     Masked Vector=<1.1 1.1 1.1>
     Passthrough Vector=&RandomRangeVector(Minimum=<0.02 0.02 0.02>, Maximum=<0.05 0.05 0.05>)
    )
    Mesh Scale Mode=::NewEnumerator4
    Sprite Rotation Angle=&RandomRangeFloat(Minimum=0, Maximum=1)
    Sprite Rotation Mode=::NewEnumerator1
    Sprite Size Mode=::NewEnumerator3
    Uniform Sprite Size=5000
    UsePositionOffset=false
    Write Material Random=true
   )
   /SetVariables(Particles.VisibilityTag=&SpawnGroup()) {isSetParameters=true, module=SetVariables_#1}
   &DynamicMaterialParameters(
    Index 0 Param 3=1
    Index 0 Param 4=1
    Module.Index0Param0Write=true
    Module.Index0Param1Write=true
    Module.Index0Param2Write=true
    Module.Index0Param3Write=true
   )
  ]
  particleUpdateScript [
   &ParticleState()
   &GravityForce(Gravity=<0 0 -100>)
   &SubUVAnimation(
    "Sprite Renderer "=DataInterface() {propertyValues {SpriteRenderer=&self:Snow_Linger_0.NiagaraSpriteRendererProperties_1}}
    SubUV Lookup Index:NiagaraFloat=&FloatFromCurve(FloatCurve=DataInterface() {propertyValues {Curve {keys [
         {arriveTangent=4.1785, interpMode=RCIM_Cubic, leaveTangent=4.1785, tangentMode=RCTM_User, time=0, value=0.15}
         {arriveTangent=0.85, interpMode=RCIM_Cubic, leaveTangent=0.85, time=0.1, value=0.7}
         {arriveTangent=0.0224, interpMode=RCIM_Cubic, leaveTangent=0.0224, tangentMode=RCTM_User, time=1, value=1}
        ]}}})
    UseEndFrame=false
    UseStartFrame=false
   ) {module=SubUVAnimation001}
   &ScaleSpriteSize(Uniform Curve Sprite Scale=DataInterface() {propertyValues {Curve {keys [
        {arriveTangent=2.6419, interpMode=RCIM_Cubic, leaveTangent=2.6419, tangentMode=RCTM_User, time=0, value=0.3}
        {interpMode=RCIM_Cubic, leaveTangent=-0.6, time=1, value=1}
       ]}}})
   /SetVariables(Particles.SnowSpriteColor:LinearColor=&MultiplyLinearColor(
     Linear Color=&ColorFromCurve(ColorCurve=DataInterface() {propertyValues {
        AlphaCurve {keys [
          <at 0.01 0>
          {arriveTangent=1.4143, interpMode=RCIM_Cubic, leaveTangent=1.4143, time=0.0848, value=0.85}
          <at 0.3635 0.5>
          <at 1 0 RCIM_Cubic>
         ]}
        BlueCurve {keys [<at 0 1 RCIM_Cubic>]}
        GreenCurve {keys [<at 0 1 RCIM_Cubic>]}
        RedCurve {keys [<at 0 1 RCIM_Cubic>]}
       }})
     Scale Factor=&MakeVector4_FromLinearColor(LinearColor=$Particles.Color)
    )) {isSetParameters=true, module=SetVariables_#2}
   &SolveForcesAndVelocity(Clamp Velocity=false, Limit Acceleration=false)
  ]
  rendererClasses [&NiagaraSpriteRendererProperties]
  renderers [{rendererClass=&NiagaraSpriteRendererProperties}]
 }
 Snow_Spike {
  emitterUpdateScript [
   &EmitterState()
   &SpawnBurst_Instantaneous(Use Spawn Probability=false, UseLoopCountLimit=false)
  ]
  particleSpawnScript [
   &InitializeParticle(
    Color=<rgba 0 0 0 0.2>
    Lifetime=1.5
    Mesh Scale=&MaskVectorBySpawnGroup(
     Masked Vector=<1.1 1.1 1.1>
     Passthrough Vector=&RandomRangeVector(Minimum=<0.02 0.02 0.02>, Maximum=<0.05 0.05 0.05>)
    )
    Mesh Scale Mode=::NewEnumerator4
    Position Offset=<0 0 7.5>
    Position Offset Coordinate Space=::World
    Sprite Rotation Angle=&RandomRangeFloat(Minimum=0, Maximum=1)
    Sprite Rotation Mode=::NewEnumerator1
    Sprite Size=<3500 8000>
    Sprite Size Mode=::NewEnumerator4
    UsePositionOffset=true
    Write Material Random=true
   ) {enabled=false}
   &InitializeParticle(
    Color=&FNTimeOfDayColor(
     Night Color=<rgba 0.31 0.3011 0.4818 1>
     Morning Color=<rgba 1 1 1 1>
     Day Color=<rgba 0.8892 0.875 1 1>
     Evening Color=<rgba 0.4818 0.3425 0.4649 1>
    )
    Lifetime=2.5
    Mesh Scale=&MaskVectorBySpawnGroup(
     Masked Vector=<1.1 1.1 1.1>
     Passthrough Vector=&RandomRangeVector(Minimum=<0.02 0.02 0.02>, Maximum=<0.05 0.05 0.05>)
    )
    Mesh Scale Mode=::NewEnumerator4
    Sprite Rotation Angle=&RandomRangeFloat(Minimum=0, Maximum=1)
    Sprite Rotation Mode=::NewEnumerator1
    Sprite Size Mode=::NewEnumerator3
    Uniform Sprite Size=5000
    UsePositionOffset=false
    Write Material Random=true
   ) {module=InitializeParticle001}
   /SetVariables(Particles.VisibilityTag=&SpawnGroup()) {isSetParameters=true, module=SetVariables_#1}
   &DynamicMaterialParameters(
    Index 0 Param 3=1
    Index 0 Param 4=1
    Module.Index0Param0Write=true
    Module.Index0Param1Write=true
    Module.Index0Param2Write=true
    Module.Index0Param3Write=true
   )
   &AddVelocity(Velocity=<0 0 500>)
  ]
  particleUpdateScript [
   &ParticleState()
   &GravityForce(Gravity=<0 0 -100>) {enabled=false}
   &SubUVAnimation(
    "Sprite Renderer "=DataInterface() {propertyValues {SpriteRenderer=&self:Snow_Spike_0.NiagaraSpriteRendererProperties_1}}
    SubUV Lookup Index:NiagaraFloat=&FloatFromCurve(FloatCurve=DataInterface() {propertyValues {Curve {keys [
         {arriveTangent=4.1785, interpMode=RCIM_Cubic, leaveTangent=4.1785, tangentMode=RCTM_User, time=0, value=0.15}
         {arriveTangent=0.85, interpMode=RCIM_Cubic, leaveTangent=0.85, time=0.1, value=0.7}
         {arriveTangent=0.0224, interpMode=RCIM_Cubic, leaveTangent=0.0224, tangentMode=RCTM_User, time=1, value=1}
        ]}}})
    UseEndFrame=false
    UseStartFrame=false
   ) {module=SubUVAnimation001}
   &ScaleSpriteSize(Uniform Curve Sprite Scale=DataInterface() {propertyValues {Curve {keys [
        {arriveTangent=2.6419, interpMode=RCIM_Cubic, leaveTangent=2.6419, tangentMode=RCTM_User, time=0, value=0.3}
        {interpMode=RCIM_Cubic, leaveTangent=-0.6, time=1, value=1}
       ]}}})
   /SetVariables(Particles.SnowSpriteColor:LinearColor=&MultiplyLinearColor(
     Linear Color=&ColorFromCurve(ColorCurve=DataInterface() {propertyValues {
        AlphaCurve {keys [
          <at 0.01 0>
          {arriveTangent=1.4143, interpMode=RCIM_Cubic, leaveTangent=1.4143, time=0.0848, value=0.85}
          <at 0.3635 0.5>
          <at 1 0 RCIM_Cubic>
         ]}
        BlueCurve {keys [<at 0 1 RCIM_Cubic>]}
        GreenCurve {keys [<at 0 1 RCIM_Cubic>]}
        RedCurve {keys [<at 0 1 RCIM_Cubic>]}
       }})
     Scale Factor=&MakeVector4_FromLinearColor(LinearColor=$Particles.Color)
    )) {isSetParameters=true, module=SetVariables_#2}
   &SolveForcesAndVelocity(Clamp Velocity=false, Limit Acceleration=false)
  ]
  rendererClasses [&NiagaraSpriteRendererProperties]
  renderers [{rendererClass=&NiagaraSpriteRendererProperties}]
 }
 Streaks {
  emitterUpdateScript [
   &EmitterState()
   &SpawnBurst_Instantaneous(Spawn Count=6, Use Spawn Probability=false)
  ]
  particleSpawnScript [
   &InitializeParticle(
    Color=<rgba 1.7121 2.7805 5 1>
    Lifetime=&UniformRangedFloat(Minimum=0.12, Maximum=0.2)
    Sprite Size Max=<140 2000>
    Sprite Size Min=<50 900>
    Sprite Size Mode=::NewEnumerator2
    UsePositionOffset=false
    Write Material Random=true
   ) {module=InitializeParticle001}
   &AddVelocity(
    Cone Angle=100
    Cone Axis=<0 0 1>
    Module.UseConeFalloff=false
    Module.UseVelDistribution=false
    Override Seed=false
    Velocity Mode=::NewEnumerator2
   )
   /SetVariables(
    Particles.DarkColor:LinearColor=&RandomRangeLinearColor(Minimum=<rgba 0.1844 0.269 0.4583 0.5>, Maximum=<rgba 0.2194 0.2965 0.3854 0.5>)
    Particles.DarkSize:Vector2f=&Multiply_Vector2D_ByFloat(Float=1.05, Vector2D=$Particles.SpriteSize)
   ) {isSetParameters=true}
  ]
  particleUpdateScript [
   &ParticleState()
   &SolveForcesAndVelocity(Clamp Velocity=false, Limit Acceleration=false)
   &ScaleSpriteSize(Scale Factor=&Vector2DFromCurve(
     Vector2Curve=DataInterface() {propertyValues {XCurve {keys [<at 0 0.75 RCIM_Cubic>,<at 1 1 RCIM_Cubic>]}, YCurve {keys [<at 0 0 RCIM_Cubic>,<at 1 2 RCIM_Cubic>]}}}
     ScaleCurve=<2 0.6>
    ))
   /SetVariables(Particles.AlphaCurve:NiagaraFloat=&FloatFromCurve(
     FloatCurve=DataInterface() {propertyValues {Curve {keys [<at 0 1 RCIM_Cubic>,<at 1 0>]}}}
     Scale Curve=&Multiply_Float(A=1, B=&FNTimeOfDayScalar(Night Scale=0.3, Morning Scale=0.6, Evening Scale=0.6))
    )) {isSetParameters=true}
   /SetVariables(Particles.DarkColor:LinearColor=&MultiplyLinearColor(
     Linear Color=$Particles.Initial.DarkColor
     Scale Factor=&MakeVector4FromVectorAndFloat(
      "Vector (RGB)"=<1 1 1>
      "Float (Alpha)"=&FloatFromCurve(
       FloatCurve=DataInterface() {propertyValues {Curve {keys [<at 0 1 RCIM_Cubic>,<at 1 0>]}}}
       Scale Curve=&FNTimeOfDayScalar(Night Scale=0.3, Morning Scale=0.6, Evening Scale=0.6)
      )
     )
    )) {isSetParameters=true}
   &ScaleColor(Scale Alpha=$Particles.AlphaCurve, Scale RGB=&VectorFromFloat(Value=1), ScaleA=true, ScaleRGB=false)
  ]
  rendererClasses [&NiagaraSpriteRendererProperties]
  renderers [
   {rendererClass=&NiagaraSpriteRendererProperties}
   {rendererClass=&NiagaraSpriteRendererProperties, rendererIndex=1}
  ]
 }
}
summary {
 systemName=NS_PC_MaskMaker_Smack_MountainImpact
 emitters [
  {emitterName=Snow_Linger, rendererClasses [&NiagaraSpriteRendererProperties]}
  {emitterName=Snow_Spike, rendererClasses [&NiagaraSpriteRendererProperties]}
  {emitterName=Fragments, simTarget=GPUComputeSim, rendererClasses [&NiagaraMeshRendererProperties]}
  {emitterName=Radial_Plasma_Energy, rendererClasses [&NiagaraSpriteRendererProperties]}
  {
   emitterName=Burst
   rendererClasses [
    &NiagaraLightRendererProperties
    &NiagaraSpriteRendererProperties
    &NiagaraMeshRendererProperties
   ]
  }
  {emitterName=Streaks, rendererClasses [&NiagaraSpriteRendererProperties]}
 ]
 userVariables [
  User.bIsFrontEnd {type=NiagaraBool}
  User.bIsFrontEndPreview {type=NiagaraBool}
  User.bIsInWater {type=NiagaraBool}
  User.ForwardDot {type=NiagaraFloat, defaultValue=0, description=""}
  User.Tint_for_FakeLight {type=LinearColor, description="", defaultValue {struct=LinearColor, value=<rgba 1 1 1 1>}}
 ]
}
